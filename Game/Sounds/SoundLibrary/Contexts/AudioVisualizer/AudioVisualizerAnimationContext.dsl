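; Sound Library context that shows the Fortnite audio visualizer indicator when a sound event plays.
; When the visualizer is enabled, Play passes the event's actor, location, sound max distance and configured type, tint and icon to ShowSoundIndicatorLocal, and also calls the parent.
; An optional enemy-only filter resolves the owner to a pawn, directly or through its weapon's owner, allowing indicators only for pawns hostile to the local player.
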(fn Play
  (params (InEventData :type Sound%20Library%20Context%20Event%20Input%20Structure%20%28by%20ref%29))
  (if (Variables|Default|GetAudioVisualizerEnabled)
    (bind _shouldfire (CallFunction|ShouldFireBasedonAffiliation InEventData))
    (if _shouldfire
      (Fortnite|ShowSoundIndicatorLocal :TrackedActor (Utilities|Struct|BreakSoundLibraryContextEventInput InEventData) :Location (Transformation|GetActorLocation (Utilities|Struct|BreakSoundLibraryContextEventInput InEventData)) :MaxAudibleDistance (Variables|/Script/Engine.SoundBase|GetMaxDistance (Utilities|Struct|BreakSoundLibraryContextEventInput InEventData)) :IndicatorType (Variables|Default|GetAudioVisualizer%20Indicator%20Type) :Instigator (Utilities|Struct|BreakSoundLibraryContextEventInput InEventData) :Tint (Variables|Default|GetAudioVisualizerTint) :OverrideIcon (Variables|Default|GetAudioVisualizerOverrideIcon) :SoundBase (Utilities|Struct|BreakSoundLibraryContextEventInput InEventData) :AttenuationMaxRangeMultiplier 1.0 :AttenuationIntensityMultiplier 1.0)
      (else
        _))
    (bind (_outcomponents _returnvalue) (|CallParentFunction InEventData))
    (return _outcomponents _returnvalue)
    (else
      _)))

(fn ShouldFireBasedOnAffiliation
  (params (EventData :type Sound%20Library%20Context%20Event%20Input%20Structure))
  (bind _eventdata (Variables|GetEventData))
  (bind _owner (Utilities|Struct|BreakSoundLibraryContextEventInput _eventdata))
  (if (Variables|Default|GetAudioVisualizerOnlyFireOnEnemy)
    (bind _asfort_pawn (Utilities|Casting|CastToFortPawn _owner)
      (:then
        (LocalVariables|SetPawn _asfort_pawn))
      (:CastFailed
        (bind _asfort_weapon (Utilities|Casting|CastToFortWeapon _owner)
          (:then
            (bind _asfort_pawn_2 (Utilities|Casting|CastToFortPawn (Actor|GetOwner _asfort_weapon))
              (:then
                (LocalVariables|SetPawn _asfort_pawn_2))
              (:CastFailed
                (return false))))
          (:CastFailed
            (return false)))))
    (switch Utilities|FlowControl|Switch|SwitchonEFortTeamAffiliation (Team|GetTeamAffiliationforActortoLocalPlayer (LocalVariables|GetPawn))
      (:Friendly
        _)
      (:Neutral
        _)
      (:Hostile
        (return true)))
    (return false)
    (else
      (return true))))
